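# Animation Blueprint for the BBQ Larry pet cosmetic, targeting the M_MED_BarbequeLarry_Backpack skeleton.
# Its event graph is empty and it declares no variables, so the file shows no custom animation logic.
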AnimBlueprint /BRCosmetics/Characters/Pets/BBQ_Larry/B_BBQ_Larry_AnimBP
target_skeleton=/Game/Accessories/FORT_Backpacks/M_MED_BarbequeLarry_Backpack/Meshes/M_MED_BarbequeLarry_Backpack_Skeleton
graphs [EventGraph {links [], nodes []}]
variables []
